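# Sound cue for the start of the Hoagie vehicle's boost.
# A random node picks one of three equally weighted boost-start wave variations.
# It plays in the exterior vehicle sound class with Hoagie engine boost attenuation.
# A concurrency override allows one instance per owner, stopping the oldest with a release fade when a new one starts.
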SoundCue /Hoagie/Audio/Assets/Hoagie_Boost_Start_Cue
root=Random_0
nodes [
 {id=Random_0, children [WavePlayer_1,WavePlayer_2,WavePlayer_3], properties {weights [1,1,1]}}
 {id=WavePlayer_1, properties {soundWaveAssetPtr=/Hoagie/Audio/Assets/Waves/Veh_Hoagie_Boost_Start_01}}
 {id=WavePlayer_2, properties {soundWaveAssetPtr=/Hoagie/Audio/Assets/Waves/Veh_Hoagie_Boost_Start_02}}
 {id=WavePlayer_3, properties {soundWaveAssetPtr=/Hoagie/Audio/Assets/Waves/Veh_Hoagie_Boost_Start_03}}
]
properties {
 attenuationSettings=/Hoagie/Audio/Hoagie_Engine_Boost_Att
 bOverrideConcurrency=true
 maxDistance=30000
 soundClassObject=&Fort_Vehicles_Exterior
 volumeMultiplier=1
 concurrencyOverrides {bLimitToOwner=true, maxCount=1, resolutionRule=StopOldest, voiceStealReleaseTime=1}
}
