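# Sound cue for critical hit damage on the Hoagie vehicle, routed through an instigator 2D hit-notification sound class.
# A mixer layers two equal-volume branches: a random Hoagie critical-hit wave and a random light metal impact.
# Each branch has its own modulator that randomly lowers pitch and volume for variation.
# Concurrency settings cap it at two simultaneous plays, stopping the oldest, and overall volume is reduced.
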SoundCue /Hoagie/Audio/Assets/Hoagie_CriticalHit_Cue
root=Mixer_0
nodes [
 {id=Mixer_0, children [Modulator_0,Modulator_1], properties {inputVolume [1,1]}}
 {id=Modulator_0, children [Random_1], properties {pitchMax=1, pitchMin=0.7, volumeMax=1, volumeMin=0.6}}
 {id=Modulator_1, children [Random_0], properties {pitchMax=1, pitchMin=0.7, volumeMax=1, volumeMin=0.6}}
 {id=Random_0, children [WavePlayer_4, WavePlayer_9, WavePlayer_10], properties {weights [1,1,1]}}
 {id=Random_1, children [WavePlayer_0,WavePlayer_1,WavePlayer_2], properties {weights [1,1,1]}}
 {id=WavePlayer_0, properties {soundWaveAssetPtr=/Hoagie/Audio/Assets/Waves/Veh_Hoagie_Dmg_Critical_Hit_01}}
 {id=WavePlayer_1, properties {soundWaveAssetPtr=/Hoagie/Audio/Assets/Waves/Veh_Hoagie_Dmg_Critical_Hit_02}}
 {id=WavePlayer_10, properties {soundWaveAssetPtr=/Game/Sounds/Fort_Weapons/Impacts/MTL_IMPACT_LIGHT_03}}
 {id=WavePlayer_2, properties {soundWaveAssetPtr=/Hoagie/Audio/Assets/Waves/Veh_Hoagie_Dmg_Critical_Hit_03}}
 {id=WavePlayer_4, properties {soundWaveAssetPtr=/Game/Sounds/Fort_Weapons/Impacts/MTL_IMPACT_LIGHT_01}}
 {id=WavePlayer_9, properties {soundWaveAssetPtr=/Game/Sounds/Fort_Weapons/Impacts/MTL_IMPACT_LIGHT_02}}
]
properties {
 maxDistance=2097152
 soundClassObject=&Fort_2DHitNotifications_Instigate_PSM
 volumeMultiplier=0.4
 concurrencyOverrides {maxCount=2, resolutionRule=StopOldest}
 concurrencySet [&HitNotifications_TwoMaximum_StopOldest_VolumeScale]
}
