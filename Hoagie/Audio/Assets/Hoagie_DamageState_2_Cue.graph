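# Sound cue for the Hoagie vehicle's second damage state, layering a damage-start sound with a looping fire and sparks bed.
# Close and distant versions of the start wave play together, each with its own engine attenuation.
# An enveloper swells the loop, then fades it low; a quality switch drops that detail layer on its last two levels.
# Concurrency allows one instance per owner, stopping the oldest.
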SoundCue /Hoagie/Audio/Assets/Hoagie_DamageState_2_Cue
root=Mixer_1
nodes [
 {id=Attenuation_1, children [WavePlayer_1], properties {attenuationSettings=/Hoagie/Audio/Hoagie_Engine_Distant_Att}}
 {id=Attenuation_4, children [WavePlayer_0], properties {attenuationSettings=/Hoagie/Audio/Hoagie_Engine_Close_Att}}
 {id=Attenuation_5, children [Enveloper_1], properties {attenuationSettings=/Hoagie/Audio/Hoagie_Engine_Close_Detail_Att}}
 {
  id=Enveloper_1
  children [WavePlayer_2]
  properties {pitchCurve {editorCurveData {keys [{time=0}]}}, volumeCurve {editorCurveData {keys [<at 0 0>,<at 3 1>,<at 20 0.15>]}}}
 }
 {id=Mixer_0, children [Attenuation_4, QualityLevel_0], properties {inputVolume [0.75,1]}}
 {id=Mixer_1, children [Mixer_0, Attenuation_1], properties {inputVolume [1,1]}}
 {id=QualityLevel_0, children [Attenuation_5, Attenuation_5, null, null]}
 {id=WavePlayer_0, properties {soundWaveAssetPtr=/Hoagie/Audio/Assets/Waves/Veh_Hoagie_Dmg_2_Start_01}}
 {id=WavePlayer_1, properties {soundWaveAssetPtr=/Hoagie/Audio/Assets/Waves/Veh_Hoagie_Dmg_2_Start_01_distant}}
 {id=WavePlayer_2, properties {bLooping=true, soundWaveAssetPtr=/Hoagie/Audio/Assets/Waves/Veh_Hoagie_Dmg_Fire_Sparks_Loop_01}}
]
properties {
 bOverrideConcurrency=true
 maxDistance=28500
 soundClassObject=&Fort_Vehicles_Exterior
 volumeMultiplier=1
 concurrencyOverrides {bLimitToOwner=true, maxCount=1, resolutionRule=StopOldest, voiceStealReleaseTime=1}
}
