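# Sound cue for the Hoagie vehicle's critical damage state, routed to the exterior vehicle sound class.
# A Mixer layers a death-dive start wave, slightly lowered and using distant attenuation, with a looping alarm using close attenuation.
# An Enveloper shapes the alarm's volume, fading it in over a few seconds, then back to silence.
# A concurrency override allows one instance per owner, stopping the oldest.
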SoundCue /Hoagie/Audio/Assets/Hoagie_DamageState_Critical_Cue
root=Mixer_0
nodes [
 {id=Attenuation_0, children [Enveloper_0], properties {attenuationSettings=/Hoagie/Audio/Hoagie_Engine_Close_Att}}
 {id=Attenuation_1, children [WavePlayer_1], properties {attenuationSettings=/Hoagie/Audio/Hoagie_Engine_Distant_Critical_Att}}
 {
  id=Enveloper_0
  children [WavePlayer_3]
  properties {pitchCurve {editorCurveData {keys [{time=0}]}}, volumeCurve {editorCurveData {keys [<at 0 0>,<at 3 1>,<at 15 0>]}}}
 }
 {id=Mixer_0, children [Attenuation_1, Attenuation_0], properties {inputVolume [0.75,1]}}
 {id=WavePlayer_1, properties {soundWaveAssetPtr=/Hoagie/Audio/Assets/Waves/Veh_Hoagie_Dmg_Death_Dive_Start_01}}
 {id=WavePlayer_3, properties {bLooping=true, soundWaveAssetPtr=/Hoagie/Audio/Assets/Waves/Veh_Hoagie_Dmg_Alarm_Loop_01}}
]
properties {
 bOverrideConcurrency=true
 maxDistance=20000
 soundClassObject=&Fort_Vehicles_Exterior
 volumeMultiplier=0.8
 concurrencyOverrides {bLimitToOwner=true, maxCount=1, resolutionRule=StopOldest, voiceStealReleaseTime=1}
}
