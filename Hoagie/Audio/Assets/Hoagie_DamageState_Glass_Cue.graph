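# Sound cue that plays a glass shatter for the Hoagie vehicle's damage state.
# An equally weighted Random node picks one of three shatter waves and feeds it through a Modulator at the root.
# It routes to the exterior vehicle sound class and borrows the golf cart's distant movement attenuation.
# A concurrency override allows one instance per owner, stopping the oldest with a voice-steal release.
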SoundCue /Hoagie/Audio/Assets/Hoagie_DamageState_Glass_Cue
root=Modulator_0
nodes [
 {id=Modulator_0, children [Random_1]}
 {id=Random_1, children [WavePlayer_0,WavePlayer_1,WavePlayer_2], properties {weights [1,1,1]}}
 {id=WavePlayer_0, properties {soundWaveAssetPtr=/Hoagie/Audio/Assets/Waves/Veh_Hoagie_Glass_Shatter_01}}
 {id=WavePlayer_1, properties {soundWaveAssetPtr=/Hoagie/Audio/Assets/Waves/Veh_Hoagie_Glass_Shatter_02}}
 {id=WavePlayer_2, properties {soundWaveAssetPtr=/Hoagie/Audio/Assets/Waves/Veh_Hoagie_Glass_Shatter_03}}
]
properties {
 attenuationSettings=/Game/Sounds/Attenuation/Fort_GolfCart_Movement_Distant_Att
 bOverrideConcurrency=true
 maxDistance=10100
 soundClassObject=&Fort_Vehicles_Exterior
 volumeMultiplier=1
 concurrencyOverrides {bLimitToOwner=true, maxCount=1, resolutionRule=StopOldest, voiceStealReleaseTime=1}
}
