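# Sound cue for the Hoagie vehicle's destruction explosion, built from close, medium and distant layers.
# A mixer plays the three waves together, each through its own distance attenuation with overridden falloff and its own sound class.
# Only the close layer passes through a modulator, left at default settings.
# The distant layer is classed as distant ranged weapon fire, while the other two use explosives classes.
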SoundCue /Hoagie/Audio/Assets/Hoagie_Destroy_Cue
root=Mixer_4
nodes [
 {
  id=Attenuation_0
  children [WavePlayer_1]
  properties {
   attenuationSettings=/Game/Sounds/Attenuation/Fort_Explosion_Distant_Athena_Att
   attenuationOverrides {falloffDistance=3000, attenuationShapeExtents {x=1000}}
  }
 }
 {
  id=Attenuation_4
  children [WavePlayer_0]
  properties {attenuationSettings=/Game/Sounds/Attenuation/Fort_Explosion_Close_Athena_Att, attenuationOverrides {falloffDistance=1600}}
 }
 {
  id=Attenuation_5
  children [WavePlayer_2]
  properties {
   attenuationSettings=/Game/Sounds/Attenuation/Fort_Explosion_Medium_Athena_Att
   attenuationOverrides {falloffDistance=3000, attenuationShapeExtents {x=1000}}
  }
 }
 {id=Mixer_4, children [SoundClass_3,SoundClass_2,SoundClass_0], properties {inputVolume [1,1,1]}}
 {id=Modulator_0, children [Attenuation_4]}
 {
  id=SoundClass_0
  children [Attenuation_0]
  properties {soundClassOverride=&Fort_WeaponRanged_Fire_Distant}
 }
 {id=SoundClass_2, children [Attenuation_5], properties {soundClassOverride=/Game/Sounds/FortSoundClassesAndModes/Fort_Explosives_Medium}}
 {id=SoundClass_3, children [Modulator_0], properties {soundClassOverride=/Game/Sounds/FortSoundClassesAndModes/Fort_Explosives_Large_PSM}}
 {id=WavePlayer_0, properties {soundWaveAssetPtr=/Hoagie/Audio/Assets/Waves/Veh_Hoagie_Expl_Close_01}}
 {id=WavePlayer_1, properties {soundWaveAssetPtr=/Hoagie/Audio/Assets/Waves/Veh_Hoagie_Expl_Distant_01}}
 {id=WavePlayer_2, properties {soundWaveAssetPtr=/Hoagie/Audio/Assets/Waves/Veh_Hoagie_Expl_Med_01}}
]
properties {maxDistance=27500, volumeMultiplier=1}
