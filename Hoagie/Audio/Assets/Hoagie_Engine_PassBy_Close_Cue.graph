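# Sound cue for the Hoagie vehicle's close engine pass-by in the exterior vehicles sound class.
# A Random node picks among four fast close pass-by wave variations with equal weight.
# A Quality Level root routes its first two slots to that same Random node and leaves the other two empty.
# Spatial falloff references a dedicated close pass-by attenuation asset, and the cue applies a slight volume boost.
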SoundCue /Hoagie/Audio/Assets/Hoagie_Engine_PassBy_Close_Cue
root=QualityLevel_0
nodes [
 {id=QualityLevel_0, children [Random_0,Random_0,null,null]}
 {id=Random_0, children [WavePlayer_1,WavePlayer_2,WavePlayer_3,WavePlayer_4], properties {weights [1,1,1,1]}}
 {id=WavePlayer_1, properties {soundWaveAssetPtr=/Hoagie/Audio/Assets/Waves/Veh_Hoagie_Pass_By_Close_Fast_04}}
 {id=WavePlayer_2, properties {soundWaveAssetPtr=/Hoagie/Audio/Assets/Waves/Veh_Hoagie_Pass_By_Close_Fast_01}}
 {id=WavePlayer_3, properties {soundWaveAssetPtr=/Hoagie/Audio/Assets/Waves/Veh_Hoagie_Pass_By_Close_Fast_02}}
 {id=WavePlayer_4, properties {soundWaveAssetPtr=/Hoagie/Audio/Assets/Waves/Veh_Hoagie_Pass_By_Close_Fast_03}}
]
properties {
 attenuationSettings=/Hoagie/Audio/Hoagie_Engine_Passby_Close_Att
 maxDistance=7096
 soundClassObject=&Fort_Vehicles_Exterior
 volumeMultiplier=1.25
 attenuationOverrides {falloffDistance=5000, attenuationShapeExtents {x=500}}
}
