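# Sound cue for the Hoagie vehicle's engine startup, classed as an exterior vehicle sound.
# A single wave player feeds the engine start recording through a mixer that halves its input volume, with a further volume reduction on the cue itself.
# It references the close engine attenuation asset and also carries its own falloff and shape override values.
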SoundCue /Hoagie/Audio/Assets/Hoagie_Engine_Startup_Cue
root=Mixer_0
nodes [
 {id=Mixer_0, children [WavePlayer_2], properties {inputVolume [0.5]}}
 {id=WavePlayer_2, properties {soundWaveAssetPtr=/Hoagie/Audio/Assets/Waves/Veh_Hoagie_Engine_Start_01}}
]
properties {
 attenuationSettings=/Hoagie/Audio/Hoagie_Engine_Close_Att
 maxDistance=12000
 soundClassObject=&Fort_Vehicles_Exterior
 volumeMultiplier=0.4
 attenuationOverrides {falloffDistance=5000, attenuationShapeExtents {x=500}}
}
