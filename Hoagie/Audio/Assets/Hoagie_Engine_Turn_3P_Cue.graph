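# Third-person sound cue for the Hoagie vehicle's engine turn, routed to the exterior vehicle sound class.
# A random close wave and a random distant wave, each picked from three, pass through their own attenuation assets and are mixed, with the close layer quieter.
# The mix gets slight random pitch modulation.
# A root quality-level node sends that same chain to its first two slots and leaves the other two empty.
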SoundCue /Hoagie/Audio/Assets/Hoagie_Engine_Turn_3P_Cue
root=QualityLevel_0
nodes [
 {id=Attenuation_0, children [Random_2], properties {attenuationSettings=/Hoagie/Audio/Hoagie_Engine_Turn_Close_Att}}
 {id=Attenuation_1, children [Random_3], properties {attenuationSettings=/Hoagie/Audio/Hoagie_Engine_Turn_Distant_Att}}
 {id=Mixer_0, children [Attenuation_0, Attenuation_1], properties {inputVolume [0.75,1]}}
 {id=Modulator_1, children [Mixer_0], properties {pitchMax=1.1, pitchMin=0.9}}
 {id=QualityLevel_0, children [Modulator_1,Modulator_1,null,null]}
 {id=Random_2, children [WavePlayer_0,WavePlayer_5,WavePlayer_6], properties {weights [1,1,1]}}
 {id=Random_3, children [WavePlayer_7,WavePlayer_8,WavePlayer_9], properties {weights [1,1,1]}}
 {id=WavePlayer_0, properties {soundWaveAssetPtr=/Hoagie/Audio/Assets/Waves/Veh_Hoagie_Turn_Close_01}}
 {id=WavePlayer_5, properties {soundWaveAssetPtr=/Hoagie/Audio/Assets/Waves/Veh_Hoagie_Turn_Close_02}}
 {id=WavePlayer_6, properties {soundWaveAssetPtr=/Hoagie/Audio/Assets/Waves/Veh_Hoagie_Turn_Close_03}}
 {id=WavePlayer_7, properties {soundWaveAssetPtr=/Hoagie/Audio/Assets/Waves/Veh_Hoagie_Turn_Distant_01}}
 {id=WavePlayer_8, properties {soundWaveAssetPtr=/Hoagie/Audio/Assets/Waves/Veh_Hoagie_Turn_Distant_02}}
 {id=WavePlayer_9, properties {soundWaveAssetPtr=/Hoagie/Audio/Assets/Waves/Veh_Hoagie_Turn_Distant_03}}
]
properties {
 maxDistance=20000
 soundClassObject=&Fort_Vehicles_Exterior
 volumeMultiplier=0.8
 attenuationOverrides {falloffDistance=5000, attenuationShapeExtents {x=500}}
}
