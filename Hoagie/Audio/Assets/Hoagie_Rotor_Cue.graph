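# Looping, attenuated exterior vehicle SoundCue for the ground rotor wash of the Hoagie vehicle.
# A wave player feeds a continuous modulator whose "Amount" parameter scales volume from a silent default, with pitch output floored at normal.
# An outer modulator uses "IsLocalPlayerInVehicle" with an inverted output range to duck the volume when the local player is aboard.
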SoundCue /Hoagie/Audio/Assets/Hoagie_Rotor_Cue
root=ModulatorContinuous_2
nodes [
 {
  id=ModulatorContinuous_0
  children [WavePlayer_2]
  properties {pitchModulationParams {minOutput=1, parameterName=Amount}, volumeModulationParams {default=0, parameterName=Amount}}
 }
 {
  id=ModulatorContinuous_2
  children [ModulatorContinuous_0]
  properties {volumeModulationParams {default=0, maxOutput=0.3, minOutput=1, parameterName=IsLocalPlayerInVehicle}}
 }
 {id=WavePlayer_2, properties {bLooping=true, soundWaveAssetPtr=/Hoagie/Audio/Assets/Waves/Veh_Hoagie_Rotor_Wash_Ground_Loop_01}}
]
properties {
 attenuationSettings=/Hoagie/Audio/Hoagie_RotorWash_Att
 maxDistance=8500
 soundClassObject=&Fort_Vehicles_Exterior
 volumeMultiplier=1
 attenuationOverrides {falloffDistance=5000, attenuationShapeExtents {x=500}}
}
