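# Sound cue for the start of a rotor blade impact on the Hoagie vehicle, routed to the exterior vehicle sound class at reduced volume.
# A quality level node feeds one blade impact wave player into its first two slots and leaves the other two empty.
# The cue references the shared rotor wash attenuation asset for spatial falloff.
# Overridden concurrency allows one instance per owner and blocks new triggers while it plays.
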SoundCue /Hoagie/Audio/Assets/Hoagie_Rotor_HitStart_Cue
root=QualityLevel_0
nodes [
 {id=QualityLevel_0, children [WavePlayer_1,WavePlayer_1,null,null]}
 {id=WavePlayer_1, properties {soundWaveAssetPtr=/Hoagie/Audio/Assets/Waves/Veh_Hoagie_Blade_Impact_Start_01}}
]
properties {
 attenuationSettings=/Hoagie/Audio/Hoagie_RotorWash_Att
 bOverrideConcurrency=true
 maxDistance=8500
 soundClassObject=&Fort_Vehicles_Exterior
 volumeMultiplier=0.2
 attenuationOverrides {falloffDistance=5000, attenuationShapeExtents {x=500}}
 concurrencyOverrides {bLimitToOwner=true, maxCount=1, resolutionRule=PreventNew}
}
